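# Conversation database for the Yeti NPC (Trog), defining the greeting and top-level branches shown when a player talks to it.
# The entry tag leads to one speech node whose default line is overridden by contextual messages for first-ever meetings, match conversation counts and specific player outfits, using curve-table priorities for some.
# The speech node then branches through link nodes to separately tagged quest, basic service, random service and Yeti dialogue conversations.
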ConversationDatabase /NPCLibrary/NPCs/Yeti/NPCLibrary_Conversation_Yeti
entry_points [{tag=Conversation.Entry.Yeti, destinations [#1]}]
nodes [
 ConversationEntryPointNode_7 {
  id=#1
  class=ConversationEntryPointNode
  outputs [#2]
  properties {entryTag {tagName=Conversation.Entry.Yeti}}
 }
 Speech_C_4 {
  id=#2
  class=Speech_C
  outputs [#3, #4, #5, #6]
  properties {
   nodeName=""
   generalConfig {
    defaultMessage=You don't seem scared of me at all. That's nice.
    contextualMessages [
     {
      message=Hi.
      contextRequirements [{requirement {class=FortControllerRequirement_IsFirstEverConversationWithNPC}}]
      priority {curve {rowName=Tandem.Speech.Priority.FirstEverInteraction}}
     }
     {
      message=You don't seem scared of me at all. That's nice.
      contextRequirements [{requirement {class=FortControllerRequirement_CheckNPCMatchConversationCount, properties {nPCUniqueIdTag {tagName=AISpawnerData.Type.Tandem.Yeti}}}}]
      priority {curve {curveTable=null, rowName=None}}
     }
     {
      message="Oh, hello."
      contextRequirements [{requirement {class=FortControllerRequirement_CheckNPCMatchConversationCount, properties {nPCUniqueIdTag {tagName=AISpawnerData.Type.Tandem.Yeti}}}}]
      priority {curve {curveTable=null, rowName=None}}
     }
     {
      message=Welcome back to the cave.
      contextRequirements [{requirement {class=FortControllerRequirement_CheckNPCMatchConversationCount, properties {nPCUniqueIdTag {tagName=AISpawnerData.Type.Tandem.Yeti}}}}]
      priority {curve {curveTable=null, rowName=None}}
     }
     {
      message=Cousin. It is good to see you.
      contextRequirements [{requirement {
         class=FortControllerRequirement_HasCID
         properties {allowedCIds [/BRCosmetics/Athena/Items/Cosmetics/Characters/CID_463_Athena_Commando_M_Hairy]}
        }}]
      priority {curve {rowName=Tandem.Speech.Priority.OutfitReaction}}
     }
     {
      message=My lady.
      contextRequirements [{requirement {
         class=FortControllerRequirement_HasCID
         properties {allowedCIds [/BRCosmetics/Athena/Items/Cosmetics/Characters/Character_IcicleCane_Bright]}
        }}]
      priority {curve {rowName=Tandem.Speech.Priority.OutfitReaction}}
     }
     {
      message=Hey. You're the king.
      contextRequirements [{requirement {
         class=FortControllerRequirement_HasCID
         properties {allowedCIds [
           /BRCosmetics/Athena/Items/Cosmetics/Characters/CID_288_Athena_Commando_M_IceKing
           /BRCosmetics/Athena/Items/Cosmetics/Characters/Character_BriskReign
          ]}
        }}]
      priority {curve {rowName=Tandem.Speech.Priority.OutfitReaction}}
     }
     {
      message=(Trog is completely starstruck.)
      contextRequirements [{requirement {class=FortControllerRequirement_HasCID, properties {allowedCIds [/BRCosmetics/Athena/Items/Cosmetics/Characters/Character_FlakeSlide]}}}]
      priority {curve {rowName=Tandem.Speech.Priority.OutfitReaction}}
     }
    ]
   }
  }
 }
 ConversationLinkNode_3 {id=#6, class=ConversationLinkNode, properties {remoteEntryTag {tagName=Conversation.Entry.Yeti.Dialogue}}}
 ConversationLinkNode_2 {id=#5, class=ConversationLinkNode, properties {remoteEntryTag {tagName=Conversation.Entry.RandomServices}}}
 ConversationLinkNode_0 {id=#4, class=ConversationLinkNode, properties {remoteEntryTag {tagName=Conversation.Entry.BasicServices}}}
 ConversationLinkNode_1 {id=#3, class=ConversationLinkNode, properties {remoteEntryTag {tagName=Conversation.Entry.NPCLibrary.Quests}}}
]
