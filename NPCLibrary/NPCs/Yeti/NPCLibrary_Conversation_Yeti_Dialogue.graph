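# Conversation database for the Yeti NPC's small-talk dialogue, entered through the Conversation.Entry.Yeti.Dialogue tag.
# The entry leads to an "Ask about..." speech node that branches into Penguins, The Ice King, The Foundation, and a Back option.
# Each topic plays a linear chain of Speech nodes, with penguins the longest and most enthusiastic.
# All three chains converge on one shared ReturnToStart node labeled BACK.
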ConversationDatabase /NPCLibrary/NPCs/Yeti/NPCLibrary_Conversation_Yeti_Dialogue
entry_points [{tag=Conversation.Entry.Yeti.Dialogue, destinations [#1]}]
nodes [
 ConversationEntryPointNode_7 {
  id=#1
  class=ConversationEntryPointNode
  outputs [#2]
  properties {entryTag {tagName=Conversation.Entry.Yeti.Dialogue}}
 }
 Speech_C_11 {
  id=#2
  class=Speech_C
  outputs [#3, #4, #5, #6]
  properties {nodeName="", softTaskIcon=&T_UI_NPC_question, generalConfig {defaultMessage=I'm not much of a talker.}}
  sub_nodes [ConversationChoiceNode_5 {class=ConversationChoiceNode, properties {defaultChoiceDisplayText=Ask about...}}]
 }
 Back_C_0 {
  id=#6
  class=Back_C
  sub_nodes [ConversationChoiceNode_7 {class=ConversationChoiceNode, properties {defaultChoiceDisplayText=BACK}}]
 }
 Speech_C_6 {
  id=#5
  class=Speech_C
  outputs [#7]
  properties {nodeName="", softTaskIcon=&T_UI_NPC_question, generalConfig {defaultMessage=Haven't met him. Sorry.}}
  sub_nodes [ConversationChoiceNode_1 {class=ConversationChoiceNode, properties {defaultChoiceDisplayText=The Foundation}}]
 }
 Speech_C_8 {
  id=#7
  class=Speech_C
  outputs [#8]
  properties {
   nodeName=""
   softTaskIcon=&T_UI_NPC_question
   generalConfig {defaultMessage=I'm more of a collector than a chatter.}
  }
 }
 ReturnToStart_C_23 {
  id=#8
  class=ReturnToStart_C
  properties {softTaskIcon=&T_UI_NPC_Back}
  sub_nodes [ConversationChoiceNode_32 {class=ConversationChoiceNode, properties {defaultChoiceDisplayText=BACK}}]
 }
 Speech_C_3 {
  id=#4
  class=Speech_C
  outputs [#9]
  properties {nodeName="", softTaskIcon=&T_UI_NPC_question, generalConfig {defaultMessage=He's the king. Of the ice.}}
  sub_nodes [ConversationChoiceNode_2 {class=ConversationChoiceNode, properties {defaultChoiceDisplayText=The Ice King}}]
 }
 Speech_C_0 {
  id=#9
  class=Speech_C
  outputs [#10]
  properties {
   nodeName=""
   softTaskIcon=&T_UI_NPC_question
   generalConfig {defaultMessage=Everything you need to know is in the name.}
  }
 }
 Speech_C_5 {
  id=#10
  class=Speech_C
  outputs [#8]
  properties {
   nodeName=""
   softTaskIcon=&T_UI_NPC_question
   generalConfig {defaultMessage=I don't really get into politics. Sorry.}
  }
 }
 Speech_C_7 {
  id=#3
  class=Speech_C
  outputs [#11]
  properties {nodeName="", softTaskIcon=&T_UI_NPC_question, generalConfig {defaultMessage="Wait, you like penguins too?!"}}
  sub_nodes [ConversationChoiceNode_0 {class=ConversationChoiceNode, properties {defaultChoiceDisplayText=Penguins}}]
 }
 Speech_C_15 {
  id=#11
  class=Speech_C
  outputs [#12]
  properties {
   nodeName=""
   softTaskIcon=&T_UI_NPC_question
   generalConfig {defaultMessage="WHOA! Okay, wait. What's your favorite kind?!"}
  }
 }
 Speech_C_14 {
  id=#12
  class=Speech_C
  outputs [#13]
  properties {
   nodeName=""
   softTaskIcon=&T_UI_NPC_question
   generalConfig {defaultMessage="I like Emperors, and Humboldts, and OH! MACARONI PENGUINS are SO CUTE!!!"}
  }
 }
 Speech_C_1 {
  id=#13
  class=Speech_C
  outputs [#8]
  properties {
   nodeName=""
   softTaskIcon=&T_UI_NPC_question
   generalConfig {defaultMessage="I could talk about penguins ALL DAY! Oh, but you're busy. I should stop. Sorry."}
  }
 }
]
